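# RadialMenuTestEntry is a CommonButtonLegacy widget in the Athena HUD context menus, apparently a placeholder entry for testing a radial menu.
# A fixed-size square SizeBox root centers a small, untinted circle image borrowed from the Friends social UI.
# It inherits the InputActionWidget from its parent and adds no variables or graph logic of its own.
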WidgetBlueprint /Game/Athena/HUD/ContextMenus/RadialMenuTestEntry
graphs [EventGraph {links [], nodes []}]
variables []
widgets {
 info {inheritedWidgetCount=1, parentClass=&CommonButtonLegacy, rootWidgetClass=&SizeBox, widgetCount=3}
 widgets [
  SizeBox_32 {
   widgetClassPath=&SizeBox
   widget=SizeBox_32
   properties {bOverride_HeightOverride=true, bOverride_WidthOverride=true, heightOverride=120, widthOverride=120}
  }
  Image_82 {
   widgetClassPath=&Image
   parent=SizeBox_32
   widget=Image_82
   properties {brush {
     bIsDynamicallyLoaded=false
     drawAs=Image
     imageType=NoImage
     mirroring=NoMirror
     resourceName=None
     resourceObject=/Game/UI/Friends_UI/Social/circle
     tiling=NoTile
     imageSize=<24 24>
     margin=<ltrb 0 0 0 0>
     outlineSettings {
      bUseBrushTransparency=false
      roundingType=HalfHeightRadius
      width=0
      color {colorUseRule=UseColor_Specified, specifiedColor=<rgba 0 0 0 0>}
      cornerRadii=<0 0 0 0>
     }
     tintColor {colorUseRule=UseColor_Specified, specifiedColor=<rgba 1 1 1 1>}
     uVRegion {bIsValid=false, max=<0 0>, min=<0 0>}
    }}
   slot {class=SizeBoxSlot, properties {horizontalAlignment=HAlign_Center, verticalAlignment=VAlign_Center}}
  }
  InputActionWidget {widgetClassPath=&CommonActionWidget, bInherited=true, widget=null}
 ]
}
